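# Unlit, translucent, two-sided glow material in the Strategic Spur pickaxe FX folder, enabled for Niagara sprites, ribbons and mesh particles.
# Particle color RGB runs through the shared FX emissive adjustment function to produce the glow color.
# A Fresnel function with a tunable exponent is multiplied by particle alpha as a view-angle-based fade.
# No textures are sampled, so color and intensity come from particle data and these functions.
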Material /BRCosmetics/Weapons/FORT_Melee/PIckaxe_StrategicSpur/FX/M_PlanePackedGlow
graph {
 properties {
  bAutomaticallySetUsageInEditor=false
  bUsedWithNiagaraMeshParticles=true
  bUsedWithNiagaraRibbons=true
  bUsedWithNiagaraSprites=true
  blendMode=BLEND_Translucent
  pixelDepthOffsetMode=PDOM_Legacy
  shadingModel=MSM_Unlit
  twoSided=true
  usedShadingModels=Unlit
 }
 nodes [
  n1 = ParticleColor()
  MaterialFunctionCall(Emissive=@n1.RGB) {materialFunction=&MF_FX_EmissiveAdjustments}
  Multiply(
   A=MaterialFunctionCall(Exponent=ScalarParameter() {defaultValue=4, parameterName="FresnelExp(S)"}) {materialFunction=&MF_Fresnel_FX}.Fresnel
   B=@n1.A
  )
 ]
}
