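# Material function for the StrategicSpur pickaxe effects that outputs a hard-edged, scrolling scanline mask.
# It transforms world position into camera space and takes the Y component, so the lines run horizontally relative to the camera rather than following mesh UVs.
# That value, scaled by ScanTiling and offset by time times ScanSpeed, drives a sine that is thresholded into thin binary bands.
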MaterialFunction /BRCosmetics/Weapons/FORT_Melee/PIckaxe_StrategicSpur/FX/MF_Scanlines
graph {nodes [
  n1 = ComponentMask(None=TransformPosition(None=WorldPosition().XYZ) {transformSourceType=TRANSFORMPOSSOURCE_World, transformType=TRANSFORMPOSSOURCE_Camera}) {g=true}
  n2 = FunctionInput() {bUsePreviewValueAsDefault=true, inputName=ScanTiling, inputType=FunctionInput_Scalar, previewValue=<0.1 0 0 1>}
  n3 = Time()
  n4 = FunctionInput() {bUsePreviewValueAsDefault=true, inputName=ScanSpeed, inputType=FunctionInput_Scalar, previewValue=<-2 0 0 1>}
  n5 = Sine(None=Add(A=Multiply(A=@n1, B=@n2) {constB=0.1}, B=Multiply(A=@n3, B=@n4) {constB=-2}))
  FunctionOutput(None=If(A=Subtract(A=@n5) {constB=0.75}, "A > B"=Constant() {r=1}, "A < B"=Constant()))
 ]}
