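# Unlit, two-sided translucent Niagara material for the trail effect of the Strategic Spur pickaxe.
# Vertical and horizontal curve-atlas gradients multiply into a color, which scanlines modulate in brightness before an emissive adjustment function.
# An alpha chain multiplies noise erosion driven by dynamic parameters, leading and trailing edge fades, a trailing-edge sphere mask, gradient alphas and particle alpha.
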Material /BRCosmetics/Weapons/FORT_Melee/PIckaxe_StrategicSpur/FX/M_StrategicSpur_Trail
graph {
 properties {
  bAutomaticallySetUsageInEditor=false
  bUsedWithNiagaraMeshParticles=true
  bUsedWithNiagaraRibbons=true
  bUsedWithNiagaraSprites=true
  blendMode=BLEND_Translucent
  pixelDepthOffsetMode=PDOM_Legacy
  shadingModel=MSM_Unlit
  twoSided=true
  usedShadingModels=Unlit
 }
 nodes [
  n1 = NamedRerouteUsage()
  n2 = MaterialFunctionCall!(ScanTiling=ScalarParameter() {defaultValue=0.3, parameterName="ScanTiling (S)"}, ScanSpeed) {materialFunction=&/MF_Scanlines}
  MaterialFunctionCall(Emissive=LinearInterpolate(A=@n1, B=Multiply(A=@n1) {constB=2}, Alpha=@n2.Result)) {materialFunction=&MF_FX_EmissiveAdjustments}
  n3 = OneMinus(None=ComponentMask(None=TextureCoordinate()) {r=true})
  Multiply(A=Multiply(A=Power(Base=@n3)), B=@n2.Result)
  n4 = TextureCoordinate()
  n5 = DynamicParameter() {defaultValue=<rgba 0 5 1 1>, paramNames [OpacityLerp, ScanLineRandom, Erosion, ErosionScale]}
  n6 = LinearInterpolate(
   A=ScalarParameter() {defaultValue=1, parameterName=ErosionScale}
   B=@n5.ErosionScale
   Alpha=ScalarParameter() {parameterName=UseErosionScaleDynParam}
  )
  n7 = Subtract(A=TextureSample(UVs=Multiply(A=@n4, B=@n6)) {texture=/Game/Athena/UI/Challenges/Art/T_UI_DataChip_Noise}.G, B=OneMinus(None=@n5.Erosion)) {constB=0.5}
  n8 = ComponentMask(None=TextureCoordinate()) {r=true}
  n9 = Multiply(
   A=Add(A=@n8, B=ScalarParameter() {group="Fade Leading && Trailing Edges", parameterName=Leading Edge Fade Bias, sortPriority=2})
   B=ScalarParameter() {defaultValue=10, group="Fade Leading && Trailing Edges", parameterName=Leading Edge Fade Scale}
  ) {constB=3}
  n10 = Multiply(
   A=Add(A=OneMinus(None=@n8), B=ScalarParameter() {group="Fade Leading && Trailing Edges", parameterName=Trailing Edge Fade Bias, sortPriority=3})
   B=ScalarParameter() {defaultValue=2, group="Fade Leading && Trailing Edges", parameterName=Trailing Edge Fade Scale, sortPriority=1}
  ) {constB=2}
  n11 = Multiply(
   A=LinearInterpolate(B=Saturate(None=@n9), Alpha=ScalarParameter() {defaultValue=1, group=Alpha, parameterName=Fade Leading Edge, sortPriority=4}) {constA=1}
   B=LinearInterpolate(B=Saturate(None=@n10), Alpha=ScalarParameter() {defaultValue=1, group=Alpha, parameterName=Fade Trailing Edge, sortPriority=5}) {constA=1}
  )
  n12 = TextureCoordinate()
  n13 = VectorParameter() {group=Trailing Edge Sphere Mask, parameterName=Trailing Edge Sphereical UV Scale, sortPriority=4, defaultValue=<rgba 0.5 1 0 0>}
  n14 = ComponentMask(None=@n13.RGB) {g=true, r=true}
  n15 = VectorParameter() {group=Trailing Edge Sphere Mask, parameterName="Trailing Edge Sphere Mask Coord (B)", sortPriority=3, defaultValue=<rgba 0.85 0.82 0 0>}
  n16 = OneMinus(None=SphereMask(
    A=Multiply(A=@n12, B=@n14)
    B=ComponentMask(None=@n15.RGB) {g=true, r=true}
    Radius=ScalarParameter() {defaultValue=0.8, group=Trailing Edge Sphere Mask, parameterName=Trailing Edge Radius, sortPriority=1}
    Hardness=ScalarParameter() {defaultValue=0.3, group=Trailing Edge Sphere Mask, parameterName=Trailing Edge Hardness, sortPriority=2}
   ))
  n17 = ScalarParameter() {defaultValue=5, group=Trailing Edge Sphere Mask, parameterName=Trailing Edge Contrast}
  n18 = CurveAtlasRowParameter(CurveTime=ComponentMask(None=TextureCoordinate()) {g=true}) {
   atlas=&/CA_StrategicSpur
   curve=&/CC_StrategicSpur1
   defaultValue=1
   parameterName=VerticalGradient
  }
  n19 = CurveAtlasRowParameter(CurveTime=ComponentMask(None=TextureCoordinate()) {r=true}) {
   atlas=&CA_StrategicSpur
   curve=&/CC_StrategicSpur
   parameterName=HorizontalGradient
  }
  n20 = Multiply(
   A=LinearInterpolate(
    B=Power(Base=@n16, Exp=@n17) {constExponent=5}
    Alpha=ScalarParameter() {defaultValue=1, group=Alpha, parameterName=Multiply Trailing Edge Sphere Mask, sortPriority=6}
   ) {constA=1}
   B=Multiply(A=@n18.A, B=@n19.A)
  )
  Multiply(A=Multiply(A=LinearInterpolate(A=@n7, Alpha=@n5.Erosion) {constAlpha=0}, B=Multiply(A=@n11, B=@n20)), B=ParticleColor().A)
  NamedRerouteDeclaration(None=Multiply(A=@n18, B=@n19)) {name=FinalColor, nodeColor=<rgba 0.8281 1 0 1>}
 ]
}
