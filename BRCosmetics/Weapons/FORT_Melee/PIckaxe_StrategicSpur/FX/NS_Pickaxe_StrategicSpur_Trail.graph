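# Niagara swing trail effect for the StrategicSpur pickaxe cosmetic, combining an anim-trail ribbon with glowing sparks.
# Trail feeds two ribbon renderers, animating material parameters by emitter age and using User.SwingSectionIndex to select per-swing alpha and glow curves.
# Glow_Sparks spawns randomly tinted sprites at a skeletal mesh socket, moved by curl noise, drag and point attraction, with time-of-day alpha scaling.
# Trail_Black and FN_Debug emitters are disabled.
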NiagaraSystem /BRCosmetics/Weapons/FORT_Melee/PIckaxe_StrategicSpur/FX/NS_Pickaxe_StrategicSpur_Trail
dependencies {
 usedDataInterfaces [&NiagaraDataInterfaceCurve, &NiagaraDataInterfaceCamera, &NiagaraDataInterfaceArrayColor]
 usedDynamicInputs []
 usedModules [
  &self:SystemSpawnScript.SystemScriptSource.SystemScriptGraph.NiagaraNodeAssignment_0.SetVariables_#1
  &SystemState
 ]
}
emitters {
 FN_Debug {
  bEnabled=false
  emitterUpdateScript [
   &EmitterState()
   &SpawnBurst_Instantaneous(Spawn Count=1, Use Spawn Probability=false)
  ]
  particleSpawnScript [/Niagara/Modules/Spawn/Initialization/InitializeParticle(
    Lifetime=5
    Module.Write Color=true
    Module.Write Lifetime=true
    Module.Write Mass=true
    Module.Write Material Random=true
    Module.Write Position=false
    Module.Write Scale=false
    Module.Write SpriteRotation=false
    Module.Write SpriteSize=true
    Position=&SimulationPosition()
    Sprite Size=<100 100>
   )]
  particleUpdateScript [
   &ParticleState()
   /SetVariables(Particles.Debug:NiagaraFloat=&MakeFloatFromInt(Int=$User.SwingSectionIndex)) {isSetParameters=true}
   &DynamicMaterialParameters(
    Index 0 Param 1=$Particles.Debug
    Module.Param0WriteEnabled=true
    Module.Param1WriteEnabled=true
    Module.Param2WriteEnabled=true
    Module.Param3WriteEnabled=true
   )
  ]
  rendererClasses [&NiagaraSpriteRendererProperties]
  renderers [{rendererClass=&NiagaraSpriteRendererProperties}]
 }
 Glow_Sparks {
  emitterUpdateScript [
   &EmitterState(
    Enable Distance Culling=true
    Enable Visibility Culling=true
    Life Cycle Mode=::NewEnumerator1
    Loop Behavior=::NewEnumerator1
    Loop Delay=0.03
    Loop Duration=2
    MaxDistance=2500
    Module.bApplySpawnCountScaleByDistance=true
    Module.bUseMaxDistance=true
    Module.bUseMinDistance=false
    Scalability Mode=::NewEnumerator1
    Scale Spawn Count=true
    UseLoopDelay=true
   )
   &SpawnRate(SpawnRate=45, Use Spawn Probability=false)
   &SpawnRate(SpawnRate=30, Use Spawn Probability=false) {module=SpawnRate001}
  ]
  particleSpawnScript [
   /Niagara/Modules/Spawn/Initialization/V2/InitializeParticle(
    Color=&SelectLinearColorFromArray(
     Array Sampling Mode=::NewEnumerator0
     Direct Array Index=?
     Color Selection Array=DataInterface() {propertyValues {ColorData [<rgba 0 1 0.6705 1>, <rgba 0.0134 0 1 1>, <rgba 0.5572 0 1 1>]}}
     Random Seed=0
    )
    Lifetime Max=0.4
    Lifetime Min=0.2
    Lifetime Mode=::NewEnumerator1
    Sprite Size=&Multiply_Vector2D_ByFloat(Float=&RandomRangeFloat(Minimum=8, Maximum=4), Vector2D=<0.3 1>)
    Sprite Size Mode=::NewEnumerator4
    UsePositionOffset=true
    Write Material Random=true
   )
   &SkeletalMeshLocation(
    Bone / Socket Index=1
    "Inherit Velocity (Scale)"=2
    Inherited Velocity Speed Limit=200
    MaskSpawnGroup=false
    Mesh Sampling Type=::NewEnumerator5
    Sampled Position Offset=<0 0 -25>
    Skeletal Mesh=$User.SkeletalMesh
    Socket Sampling Mode=::NewEnumerator5
    Velocity Sampling=::NewEnumerator0
    VelocitySpeedLimit=true
   )
   &ShapeLocation(MaskSpawnGroup=false, Override Seed=false, Sphere Radius=10, Transform Order:ENiagara_TransformOrder=::NewEnumerator1, UseOwnerScale=true) {module=ShapeLocation001}
   &AddVelocity(
    Module.Use Falloff Distance=false
    Override Seed=false
    Velocity Mode=::NewEnumerator1
    Velocity Speed=&RandomRangeFloat(Minimum=50, Maximum=50)
   ) {enabled=false, module=AddVelocity001}
  ]
  particleUpdateScript [
   &ParticleState()
   &Drag(Drag=10, Use Linear Drag=true, Use Rotational Drag=true)
   &CurlNoiseForce(Noise Strength=200, Pan Noise=false)
   &PointAttractionForce(Kill Within Radius=false, Module.Use Falloff=true, UseAttractorOffset=false)
   &SolveForcesAndVelocity(Clamp Velocity=false, Limit Acceleration=false)
   &ScaleSpriteSize(Uniform Curve Sprite Scale=DataInterface() {propertyValues {Curve {keys [
        {tangentMode=RCTM_User, time=0, value=1}
        {arriveTangent=-0.0405, interpMode=RCIM_Cubic, leaveTangent=-0.0405, tangentMode=RCTM_User, time=0.7009, value=1}
        {arriveTangent=-8.0089, interpMode=RCIM_Cubic, leaveTangent=-8.0089, tangentMode=RCTM_User, time=1, value=0}
       ]}}})
   &ScaleColor(
    Scale Alpha=&Multiply_Float(
     A=&FloatFromCurve(FloatCurve=DataInterface() {propertyValues {Curve {keys [<at 0 1>,<at 1 0>]}}})
     B=&FNTimeOfDayScalar(Night Scale=0.5, Morning Scale=0.75, Evening Scale=0.75)
    )
    ScaleA=true
    ScaleRGB=true
   )
  ]
  rendererClasses [&NiagaraSpriteRendererProperties]
  renderers [{rendererClass=&NiagaraSpriteRendererProperties}]
 }
 Trail {
  emitterSpawnScript [
   &FN_RibbonIDOffsetUtility()
   /SetVariables(Emitter.TrailSpawnGroup:NiagaraInt32=&RandomRangeInt(Minimum=0, Maximum=10000)) {isSetParameters=true}
  ]
  emitterUpdateScript [
   &EmitterState(Life Cycle Mode=::NewEnumerator1, Loop Behavior=::NewEnumerator1, Loop Duration=1, UseLoopDelay=true)
   &FN_SpawnAnimTrail_v2()
  ]
  particleSpawnScript [
   /Game/Effects/Niagara/Modules/AnimTrail/FN_InitializeAnimTrail/FN_InitializeAnimTrailBoneIndecies(Second Bone or Socket Index=1)
   /Game/Effects/Niagara/Modules/AnimTrail/FN_InitializeAnimTrail/FN_InitializeAnimTrail(AnimTrailWidth=0.8, Color=<rgba 3 3 3 2>, Lifetime=0.3)
   &SystemLocation(MaskSpawnGroup=false)
   /SetVariables(Particles.SpawnGroup=$Emitter.TrailSpawnGroup) {isSetParameters=true}
  ]
  particleUpdateScript [
   &FN_UpdateAnimTrail_v2()
   &DynamicMaterialParameters(
    Index 0 Param 1=&FloatFromCurve(
     FloatCurve=DataInterface() {propertyValues {Curve {keys [
         {arriveTangent=0.0745, interpMode=RCIM_Cubic, leaveTangent=7.9847, tangentMode=RCTM_Break, time=0.2, value=0}
         {arriveTangent=0.0616, interpMode=RCIM_Cubic, leaveTangent=0.0616, tangentMode=RCTM_User, time=0.4, value=1}
        ]}}}
     CurveIndex=$Emitter.Age
    )
    Index 0 Param 2=&FloatFromCurve(
     FloatCurve=DataInterface() {propertyValues {Curve {keys [{arriveTangent=3.2789, time=0.05, value=-0.5}, {arriveTangent=7.5315, leaveTangent=7.5315, time=0.35, value=1.2}]}}}
     CurveIndex=$Emitter.Age
    )
    Index 0 Param 3=&FloatFromCurve(FloatCurve=DataInterface() {propertyValues {Curve {keys [
         {arriveTangent=-2.5557, interpMode=RCIM_Cubic, leaveTangent=-2.5557, tangentMode=RCTM_User, time=0, value=1}
         {interpMode=RCIM_Cubic, tangentMode=RCTM_User, time=0.8, value=0}
        ]}}})
    Module.Index0Param0Write=false
    Module.Index0Param1Write=true
    Module.Index0Param2Write=true
    Module.Index0Param3Write=false
   )
   &ScaleColor(
    Scale Alpha=&self:SectionIndexFloatSelector(
     float:NiagaraFloat=&Multiply_Float(
      A=&FloatFromCurve(
       FloatCurve=DataInterface() {propertyValues {Curve {keys [
           {arriveTangent=-11.3664, interpMode=RCIM_Cubic, tangentMode=RCTM_User, time=0, value=0.6}
           {interpMode=RCIM_Cubic, tangentMode=RCTM_User, time=0.15, value=1}
           {arriveTangent=0.5919, interpMode=RCIM_Cubic, leaveTangent=0.5919, tangentMode=RCTM_User, time=0.3, value=0}
          ]}}}
       CurveIndex=$Emitter.NormalizedLoopAge
      )
      B=1
     )
     float001:NiagaraFloat=&Multiply_Float(
      A=&FloatFromCurve(
       FloatCurve=DataInterface() {propertyValues {Curve {keys [
           {arriveTangent=-11.3664, interpMode=RCIM_Cubic, tangentMode=RCTM_User, time=0, value=0.6}
           {interpMode=RCIM_Cubic, tangentMode=RCTM_User, time=0.15, value=1}
           {arriveTangent=0.5919, interpMode=RCIM_Cubic, leaveTangent=0.5919, tangentMode=RCTM_User, time=0.23, value=0}
          ]}}}
       CurveIndex=$Emitter.NormalizedLoopAge
      )
      B=1
     )
     float002:NiagaraFloat=&Multiply_Float(
      A=&FloatFromCurve(
       FloatCurve=DataInterface() {propertyValues {Curve {keys [
           {arriveTangent=-11.3664, interpMode=RCIM_Cubic, tangentMode=RCTM_User, time=0, value=0.6}
           {interpMode=RCIM_Cubic, tangentMode=RCTM_User, time=0.15, value=1}
           {arriveTangent=0.5919, interpMode=RCIM_Cubic, leaveTangent=0.5919, tangentMode=RCTM_User, time=0.25, value=0}
          ]}}}
       CurveIndex=$Emitter.NormalizedLoopAge
      )
      B=1
     )
     float003:NiagaraFloat=&Multiply_Float(
      A=&FloatFromCurve(
       FloatCurve=DataInterface() {propertyValues {Curve {keys [
           {arriveTangent=-11.3664, interpMode=RCIM_Cubic, tangentMode=RCTM_User, time=0, value=0.6}
           {interpMode=RCIM_Cubic, tangentMode=RCTM_User, time=0.1, value=1}
           {arriveTangent=0.5919, interpMode=RCIM_Cubic, leaveTangent=0.5919, tangentMode=RCTM_User, time=0.2, value=0}
          ]}}}
       CurveIndex=$Emitter.NormalizedLoopAge
      )
      B=1
     )
     SwingIndex:NiagaraInt32=$User.SwingSectionIndex
    )
    ScaleA=true
    ScaleRGB=true
   )
   /SetVariables(
    Particles.GlowColor=&MultiplyLinearColor(
     Linear Color=<rgba 1 1 1 1>
     Scale Factor=&MakeVector4(
      X=1
      Y=1
      Z=1
      W=&Multiply_Float(
       A=&FloatFromCurve(FloatCurve=DataInterface() {propertyValues {Curve {keys [
            {interpMode=RCIM_Cubic, tangentMode=RCTM_User, time=0.04, value=0}
            {interpMode=RCIM_Cubic, tangentMode=RCTM_User, time=0.15, value=1}
            {interpMode=RCIM_Cubic, tangentMode=RCTM_User, time=1, value=0}
           ]}}})
       B=&self:SectionIndexFloatSelector(
        float:NiagaraFloat=&Multiply_Float(
         A=&FloatFromCurve(
          FloatCurve=DataInterface() {propertyValues {Curve {keys [
              {arriveTangent=-11.3664, interpMode=RCIM_Cubic, tangentMode=RCTM_User, time=0.05, value=0}
              {interpMode=RCIM_Cubic, tangentMode=RCTM_User, time=0.15, value=1}
              {arriveTangent=0.5919, interpMode=RCIM_Cubic, leaveTangent=0.5919, tangentMode=RCTM_User, time=0.3, value=0}
             ]}}}
          CurveIndex=$Emitter.NormalizedLoopAge
         )
         B=1
        )
        float001:NiagaraFloat=&Multiply_Float(
         A=&FloatFromCurve(
          FloatCurve=DataInterface() {propertyValues {Curve {keys [
              {arriveTangent=-11.3664, interpMode=RCIM_Cubic, tangentMode=RCTM_User, time=0.07, value=0.6}
              {interpMode=RCIM_Cubic, tangentMode=RCTM_User, time=0.15, value=1}
              {arriveTangent=0.5919, interpMode=RCIM_Cubic, leaveTangent=0.5919, tangentMode=RCTM_User, time=0.23, value=0}
             ]}}}
          CurveIndex=$Emitter.NormalizedLoopAge
         )
         B=1
        )
        float002:NiagaraFloat=&Multiply_Float(
         A=&FloatFromCurve(
          FloatCurve=DataInterface() {propertyValues {Curve {keys [
              {arriveTangent=-11.3664, interpMode=RCIM_Cubic, tangentMode=RCTM_User, time=0.07, value=0}
              {interpMode=RCIM_Cubic, tangentMode=RCTM_User, time=0.15, value=1}
              {arriveTangent=0.5919, interpMode=RCIM_Cubic, leaveTangent=0.5919, tangentMode=RCTM_User, time=0.25, value=0}
             ]}}}
          CurveIndex=$Emitter.NormalizedLoopAge
         )
         B=1
        )
        float003:NiagaraFloat=&Multiply_Float(
         A=&FloatFromCurve(
          FloatCurve=DataInterface() {propertyValues {Curve {keys [
              {arriveTangent=-11.3664, interpMode=RCIM_Cubic, tangentMode=RCTM_User, time=0, value=0.6}
              {interpMode=RCIM_Cubic, tangentMode=RCTM_User, time=0.07, value=1}
              {arriveTangent=0.5919, interpMode=RCIM_Cubic, leaveTangent=0.5919, tangentMode=RCTM_User, time=0.15, value=0}
             ]}}}
          CurveIndex=$Emitter.NormalizedLoopAge
         )
         B=1
        )
        SwingIndex:NiagaraInt32=$User.SwingSectionIndex
       )
      )
     )
    )
    Particles.GlowRibbonWidth=50
   ) {isSetParameters=true}
   /SetVariables(Particles.RibbonFacingGlow:Vector3f=&Lerp_Vector(
     A=&SubtractVectorFromPosition(Position=Linked() {linkedVariable {name=Particles.SecondBonePosition, type=Vector3f}}, Vector=$Particles.FirstBonePosition)
     B=&ReturnCameraVector(Camera Vectors=::NewEnumerator1, Coordinate Space=::Local)
     Alpha=&FloatFromCurve(FloatCurve=DataInterface() {propertyValues {Curve {keys [{interpMode=RCIM_Cubic, tangentMode=RCTM_User, time=0.03, value=0}, {interpMode=RCIM_Cubic, tangentMode=RCTM_User, time=0.1, value=1}]}}})
    )) {enabled=false, isSetParameters=true}
  ]
  rendererClasses [&NiagaraRibbonRendererProperties]
  renderers [
   {rendererClass=&NiagaraRibbonRendererProperties}
   {rendererClass=&NiagaraRibbonRendererProperties, rendererIndex=1}
  ]
 }
 Trail_Black {
  bEnabled=false
  emitterUpdateScript [
   &EmitterState(Life Cycle Mode=::NewEnumerator1, Loop Behavior=::NewEnumerator1, Loop Duration=0.27, UseLoopDelay=true)
   &FN_SpawnAnimTrail_v2()
  ]
  particleSpawnScript [
   &FN_InitializeAnimTrailBoneIndecies(Second Bone or Socket Index=1)
   &FN_InitializeAnimTrail(AnimTrailWidth=1, Color=$System.Color Dark, Lifetime=0.15)
   &SystemLocation(MaskSpawnGroup=false)
  ]
  particleUpdateScript [
   &ScaleRibbonWidth(Ribbon Width Scale=&FloatFromCurve(FloatCurve=DataInterface() {propertyValues {Curve {keys [
         {arriveTangent=-1.2579, interpMode=RCIM_Cubic, time=0, value=1}
         {arriveTangent=-0.9664, interpMode=RCIM_Cubic, leaveTangent=-0.9664, tangentMode=RCTM_User, time=1, value=0}
        ]}}}))
   &ScaleColor(
    Scale Alpha=&Multiply_Float(
     A=&FloatFromCurve(FloatCurve=DataInterface() {propertyValues {Curve {keys [
          {interpMode=RCIM_Cubic, leaveTangent=-2.3646, tangentMode=RCTM_Break, time=0, value=1}
          <at 0.6689 0.0233>
          {arriveTangent=-0.2139, interpMode=RCIM_Cubic, tangentMode=RCTM_Break, time=1, value=0}
         ]}}})
     B=&FNTimeOfDayScalar(Night Scale=0.5, Morning Scale=0.75, Evening Scale=0.75)
    )
    ScaleA=true
    ScaleRGB=true
   ) {module=ScaleColor002}
   &FN_UpdateAnimTrail_v2()
  ]
  rendererClasses [&NiagaraRibbonRendererProperties]
  renderers [{rendererClass=&NiagaraRibbonRendererProperties}]
 }
}
summary {
 systemName=NS_Pickaxe_StrategicSpur_Trail
 emitters [
  {emitterName=Trail, rendererClasses [&NiagaraRibbonRendererProperties]}
  {bEnabled=false, emitterName=Trail_Black, rendererClasses [&NiagaraRibbonRendererProperties]}
  {emitterName=Glow_Sparks, rendererClasses [&NiagaraSpriteRendererProperties]}
  {bEnabled=false, emitterName=FN_Debug, rendererClasses [&NiagaraSpriteRendererProperties]}
 ]
 userVariables [
  User.bIsFrontEnd {type=NiagaraBool}
  User.bIsFrontEndPreview {type=NiagaraBool}
  User.ParaGlideLeanAlpha {type=NiagaraFloat, defaultValue=0, description=""}
  User.RibbonIDOffset {type=NiagaraInt32}
  User.SkeletalMesh {
   type=NiagaraDataInterfaceSkeletalMesh
   defaultValue {
    struct=NiagaraExt_VariableValue_DataInterface
    value {dataInterface=&self:NiagaraDataInterfaceSkeletalMesh_0}
   }
  }
  User.SwingSectionIndex {type=NiagaraInt32, defaultValue=1}
 ]
}
